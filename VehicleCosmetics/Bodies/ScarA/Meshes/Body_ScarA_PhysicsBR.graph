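# Physics asset for the ScarA vehicle body mesh in the vehicle cosmetics content, likely the battle royale variant given its BR suffix.
# It defines a single body on the root bone with no collision shapes listed.
# It has no constraints, so the asset describes no articulated or jointed parts.
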PhysicsAsset /VehicleCosmetics/Bodies/ScarA/Meshes/Body_ScarA_PhysicsBR
bodies {root []}
constraints []
